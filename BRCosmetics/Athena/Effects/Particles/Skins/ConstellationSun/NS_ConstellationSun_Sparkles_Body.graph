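# Niagara system of faint blue body sparkles for the ConstellationSun character skin.
# Two active emitters sporadically spawn short-lived, camera-offset sprites inside a small X-axis cylinder.
# IdleFlames has two sprite renderers and separate StarGlow color and size parameters, with a curve-driven flickering size and fading glow alpha.
# IdleFlames002 adds smaller specks that fade and shrink, while the OmnidirectionalBurst emitter is disabled.
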NiagaraSystem /BRCosmetics/Athena/Effects/Particles/Skins/ConstellationSun/NS_ConstellationSun_Sparkles_Body
dependencies {
 usedDataInterfaces [&NiagaraDataInterfaceCurve, &NiagaraDataInterfaceVector2DCurve]
 usedDynamicInputs []
 usedModules [&SystemState]
}
emitters {
 IdleFlames {
  emitterSpawnScript [&FN_PreserveSkydiveUserVariables() {module=PreserveUserVariables}]
  emitterUpdateScript [
   &EmitterState()
   &SpawnRate(Spawn Probability=0.2, SpawnRate=2, Use Spawn Probability=true)
  ]
  particleSpawnScript [
   /Niagara/Modules/Spawn/Initialization/InitializeParticle(
    Color=<rgba 0 0.3286 0.4948 1>
    Lifetime=&UniformRangedFloat(Minimum=0.9, Maximum=1.25)
    Module.Write Color=true
    Module.Write Lifetime=true
    Module.Write Mass=false
    Module.Write Material Random=true
    Module.Write Position=false
    Module.Write Scale=false
    Module.Write SpriteRotation=false
    Module.Write SpriteSize=true
    Sprite Size=<4 4>
   )
   /SetVariables(Particles.StarGlowColor:LinearColor=<rgba 0 0.6642 1 0.075>, Particles.StarGlowSpriteSize:Vector2f=<6 6>) {isSetParameters=true, module=SetVariables_#1}
   &CylinderLocation(
    Cylinder Height=20
    Cylinder Radius=10
    MaskSpawnGroup=false
    Module.Enable Lathe Profile=false
    Module.Override Randomness=false
    Module.Surface Only=false
    Orientation Axis:ENiagaraOrientationAxis=::XAxis
   )
  ]
  particleUpdateScript [
   /Niagara/Modules/Update/Velocity/VortexVelocity(Module.Use Influence Falloff=false, Velocity Amount=5) {enabled=false}
   &ParticleState()
   &SolveForcesAndVelocity(Clamp Velocity=false, Limit Acceleration=false)
   /SetVariables(
    Particles.StarGlowColor:LinearColor=&MultiplyLinearColor(
     Linear Color=$Particles.Initial.StarGlowColor
     Scale Factor=&MakeVector4(
      X=1
      Y=1
      Z=1
      W=&FloatFromCurve(FloatCurve=DataInterface() {propertyValues {Curve {keys [<at 0 1>,<at 1 0>]}}})
     )
    )
    Particles.StarGlowSpriteSize:Vector2f=&Multiply_Vector2(
     A=$Particles.Initial.StarGlowSpriteSize
     B=&Vector2DFromCurve(
      Vector2Curve=DataInterface() {propertyValues {
        XCurve {keys [
          {arriveTangent=50.4225, interpMode=RCIM_Cubic, leaveTangent=50.4232, tangentMode=RCTM_User, time=0, value=0.75}
          {arriveTangent=1, interpMode=RCIM_Cubic, leaveTangent=1, time=0.1, value=0.28}
          {arriveTangent=1.3585, interpMode=RCIM_Cubic, leaveTangent=1.3585, time=0.25, value=1}
          {arriveTangent=-1.3333, interpMode=RCIM_Cubic, leaveTangent=-1.3333, time=0.63, value=1}
          <at 1 0>
         ]}
        YCurve {keys [
          <at 0 2 RCIM_Cubic>
          {arriveTangent=-4, interpMode=RCIM_Cubic, leaveTangent=-4, time=0.1, value=1}
          <at 0.25 1>
          <at 0.2615 1 RCIM_Cubic>
          {arriveTangent=-1.354, interpMode=RCIM_Cubic, leaveTangent=-1.354, time=0.63, value=1}
          <at 1 0>
         ]}
       }}
      ScaleCurve=<1 1>
     )
    )
   ) {isSetParameters=true}
   &ScaleSpriteSize(Scale Factor=&Vector2DFromCurve(
     Vector2Curve=DataInterface() {propertyValues {
       XCurve {keys [
         {arriveTangent=50.4225, interpMode=RCIM_Cubic, leaveTangent=50.4232, tangentMode=RCTM_User, time=0, value=0.75}
         {arriveTangent=1, interpMode=RCIM_Cubic, leaveTangent=1, time=0.1, value=0.28}
         {arriveTangent=1.3585, interpMode=RCIM_Cubic, leaveTangent=1.3585, time=0.25, value=1}
         {arriveTangent=-1.3333, interpMode=RCIM_Cubic, leaveTangent=-1.3333, time=0.63, value=1}
         <at 1 0>
        ]}
       YCurve {keys [
         <at 0 2 RCIM_Cubic>
         {arriveTangent=-4, interpMode=RCIM_Cubic, leaveTangent=-4, time=0.1, value=1}
         <at 0.25 1>
         <at 0.2615 1 RCIM_Cubic>
         {arriveTangent=-1.354, interpMode=RCIM_Cubic, leaveTangent=-1.354, time=0.63, value=1}
         <at 1 0>
        ]}
      }}
     ScaleCurve=<1 1>
    ))
   &CameraOffset(Camera Offset Amount=50)
  ]
  rendererClasses [&NiagaraSpriteRendererProperties]
  renderers [
   {rendererClass=&NiagaraSpriteRendererProperties}
   {rendererClass=&NiagaraSpriteRendererProperties, rendererIndex=1}
  ]
 }
 IdleFlames002 {
  emitterSpawnScript [&FN_PreserveSkydiveUserVariables() {module=PreserveUserVariables}]
  emitterUpdateScript [
   &EmitterState(
    Life Cycle Mode=::NewEnumerator1
    Loop Duration=&UniformRangedFloat(Minimum=0, Maximum=1)
    UseLoopDelay=false
   )
   &SpawnRate(Spawn Probability=0.2, SpawnRate=3, Use Spawn Probability=true)
  ]
  particleSpawnScript [
   &InitializeParticle(
    Color=<rgba 0 0.3286 0.4948 1>
    Lifetime=&UniformRangedFloat(Minimum=0.9, Maximum=1.25)
    Module.Write Color=true
    Module.Write Lifetime=true
    Module.Write Mass=false
    Module.Write Material Random=true
    Module.Write Position=false
    Module.Write Scale=false
    Module.Write SpriteRotation=false
    Module.Write SpriteSize=true
    Sprite Size=&Vector2DFromFloat(Value=&UniformRangedFloat(Minimum=1, Maximum=2))
   )
   &CylinderLocation(
    Cylinder Height=20
    Cylinder Radius=10
    Lathe Profile=DataInterface() {propertyValues {Curve {keys [<at 0 0.451 RCIM_Cubic>, {arriveTangent=0.049, interpMode=RCIM_Cubic, leaveTangent=0.049, time=0.5, value=1}, <at 1 0.5 RCIM_Cubic>]}}}
    MaskSpawnGroup=false
    Module.Enable Lathe Profile=false
    Module.Override Randomness=false
    Module.Surface Only=false
    Orientation Axis:ENiagaraOrientationAxis=::XAxis
   )
  ]
  particleUpdateScript [
   &ParticleState()
   &SolveForcesAndVelocity(Clamp Velocity=false, Limit Acceleration=false)
   &ScaleSpriteSize(Scale Factor=&Vector2DFromFloat(Value=&FloatFromCurve(FloatCurve=DataInterface() {propertyValues {Curve {keys [<at 0.5 1>,<at 1 0>]}}})))
   &ScaleColor(
    Scale Alpha=&FloatFromCurve(FloatCurve=DataInterface() {propertyValues {Curve {keys [<at 0 0>,<at 0.1 1>,<at 1 0>]}}})
    ScaleA=true
    ScaleRGB=true
   )
   &CameraOffset(Camera Offset Amount=50)
  ]
  rendererClasses [&NiagaraSpriteRendererProperties]
  renderers [{rendererClass=&NiagaraSpriteRendererProperties}]
 }
 OmnidirectionalBurst {
  bEnabled=false
  emitterUpdateScript [
   &EmitterState()
   &SpawnBurst_Instantaneous(Spawn Count=75, Use Spawn Probability=false)
  ]
  particleSpawnScript [
   &InitializeParticle(
    Lifetime=&UniformRangedFloat(Minimum=1, Maximum=2.5)
    Mass=&UniformRangedFloat(Minimum=0.4, Maximum=2)
    Module.Write Color=true
    Module.Write Lifetime=true
    Module.Write Mass=true
    Module.Write Material Random=true
    Module.Write Position=false
    Module.Write Scale=false
    Module.Write SpriteRotation=true
    Module.Write SpriteSize=false
    Position=&SimulationPosition()
    Sprite Rotation=&NormalizedAngleToDegrees(NormalizedAngle=&UniformRangedFloat(Minimum=0, Maximum=1))
   )
   /Niagara/Modules/Update/Mass/CalculateSizeAndRotationalInertiaByMass(Module.Manually Enter Density=false) {module=CalculateSizeByMass}
   /Niagara/Modules/Spawn/Location/SphereLocation(MaskSpawnGroup=false, Module.Override Randomness=false, Sphere Radius=40, Surface Only=false)
   &AddVelocityFromPoint(
    Module.Use Falloff Distance=false
    Origin Offset=<0 0 -7>
    Velocity Origin:Vector3f=/Niagara/DynamicInputs/Helpers/SimulationPosition()
    Velocity Strength=&UniformRangedFloat(Minimum=75, Maximum=400)
   )
  ]
  particleUpdateScript [
   &ParticleState() {module=UpdateAge}
   &GravityForce()
   &Drag(Drag=0.75, Use Linear Drag=true, Use Rotational Drag=false)
   &ScaleSpriteSize(Scale Factor=&Vector2DFromFloat(Value=&FloatFromCurve(FloatCurve=DataInterface() {propertyValues {Curve {keys [
          <at -0.0046 1.0113 RCIM_Cubic>
          {arriveTangent=-1.0036, interpMode=RCIM_Cubic, leaveTangent=-1.0036, time=0.7001, value=0.872}
          <at 1.0011 0.002>
         ]}}}))) {module=SpriteSizeScale}
   &ScaleColor(
    Scale Alpha=&FloatFromCurve(FloatCurve=DataInterface() {propertyValues {Curve {keys [<at -0.0046 1.0113 RCIM_Cubic>,<at 1.0011 0.002 RCIM_Cubic>]}}})
    ScaleA=true
    ScaleRGB=true
   )
   &SolveForcesAndVelocity(Clamp Velocity=false, Limit Acceleration=false)
  ]
  rendererClasses [&NiagaraSpriteRendererProperties]
  renderers [{rendererClass=&NiagaraSpriteRendererProperties}]
 }
}
summary {
 systemName=NS_ConstellationSun_Sparkles_Body
 emitters [
  {bEnabled=false, emitterName=OmnidirectionalBurst, rendererClasses [&NiagaraSpriteRendererProperties]}
  {emitterName=IdleFlames, rendererClasses [&NiagaraSpriteRendererProperties]}
  {emitterName=IdleFlames002, rendererClasses [&NiagaraSpriteRendererProperties]}
 ]
 userVariables [
  User.bIsFrontEndPreview {type=NiagaraBool}
  User.PlayerSpeed {type=NiagaraFloat, defaultValue=0, description=""}
  User.SkeletalMesh {
   type=NiagaraDataInterfaceSkeletalMesh
   defaultValue {
    struct=NiagaraExt_VariableValue_DataInterface
    value {dataInterface=&self:NiagaraDataInterfaceSkeletalMesh_0}
   }
  }
 ]
}
